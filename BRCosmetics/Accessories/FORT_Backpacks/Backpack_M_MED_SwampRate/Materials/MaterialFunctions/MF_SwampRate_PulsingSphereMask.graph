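# Material function for the SwampRate backpack that produces a pulsing, screen-aligned spherical mask around a local-space center.
# View-space pixel and center positions are combined by depth ratio into center-relative 2D coordinates, whose radius-normalized length forms the falloff.
# A time-driven sine, replaceable by a debug override, blends the falloff's exponent and multiplier.
# It outputs mask, pulse and UV, and also declares a time-of-day color from four parameters.
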MaterialFunction /BRCosmetics/Accessories/FORT_Backpacks/Backpack_M_MED_SwampRate/Materials/MaterialFunctions/MF_SwampRate_PulsingSphereMask
graph {nodes [
  n1 = FunctionInput() {
   bUsePreviewValueAsDefault=true
   description="for debugging purpose | [0, 1] : pulse sine override value; <0: no override;"
   inputName=Pulse Override
   inputType=FunctionInput_Scalar
   sortPriority=70
   previewValue=<-1 0 0 1>
  }
  n2 = Time()
  n3 = FunctionInput() {bUsePreviewValueAsDefault=true, inputName=Pulse Phase, inputType=FunctionInput_Scalar, sortPriority=20, previewValue=<0.4 0 0 1>}
  n4 = ConstantBiasScale(None=Sine(None=Multiply(A=@n2, B=@n3)))
  FunctionOutput(None=NamedRerouteDeclaration(None=If(A=@n1, "A > B"=@n1, "A < B"=@n4)) {name="Pulse Sine (S)", nodeColor=<rgba 0 1 0.6953 1>}) {outputName="Pulse (S)", sortPriority=20}
  MaterialExpressionTransformPosition_7 = TransformPosition(None=LocalPosition().XYZ) {transformType=TRANSFORMPOSSOURCE_View}
  n5 = Convert(None=@MaterialExpressionTransformPosition_7) {
   convertInputs [{
     type=Vector3
     expressionInput {
      expression=&self:MaterialExpressionTransformPosition_7
      inputName=None
      mask=0
      maskA=0
      maskB=0
      maskG=0
      maskR=0
      outputIndex=0
     }
    }]
   convertOutputs [{type=Vector2}, {type=Scalar}]
  }
  MaterialExpressionTransformPosition_8 = TransformPosition(None=FunctionInput() {bUsePreviewValueAsDefault=true, inputName=Sphere Center LS, previewValue=<0.0016 -13.43 -1.95 1>}) {transformType=TRANSFORMPOSSOURCE_View}
  n6 = Convert(None=@MaterialExpressionTransformPosition_8) {
   convertInputs [{
     type=Vector3
     expressionInput {
      expression=&self:MaterialExpressionTransformPosition_8
      inputName=None
      mask=0
      maskA=0
      maskB=0
      maskG=0
      maskR=0
      outputIndex=0
     }
    }]
   convertOutputs [{type=Vector2}, {type=Scalar}]
  }
  n7 = Multiply(A=@n5, B=Divide(A=@n6, B=@n5))
  FunctionOutput(None=NamedRerouteDeclaration(None=Subtract(A=@n7, B=@n6)) {name="Screen Aligned UV (V2)", nodeColor=<rgba 1 0 0.3047 1>}) {outputName="Screen Aligned UV (V2)", sortPriority=10}
  n8 = NamedRerouteUsage()
  n9 = FunctionInput() {bUsePreviewValueAsDefault=true, inputName=Sphere Radius, inputType=FunctionInput_Scalar, sortPriority=10, previewValue=<4.686 0 0 1>}
  n10 = Saturate(None=Divide(A=Length(None=@n8) {is=class MaterialExpressionLength}, B=@n9))
  n11 = FunctionInput() {bUsePreviewValueAsDefault=true, inputName=Min Pulse - Sphere Exp, inputType=FunctionInput_Scalar, sortPriority=30, previewValue=<3 0 0 1>}
  n12 = FunctionInput() {bUsePreviewValueAsDefault=true, inputName=Max Pulse - Sphere Exp, inputType=FunctionInput_Scalar, sortPriority=40, previewValue=<1.5 0 0 1>}
  n13 = NamedRerouteUsage()
  n14 = FunctionInput() {bUsePreviewValueAsDefault=true, inputName=Min Pulse - Sphere Mult, inputType=FunctionInput_Scalar, sortPriority=50, previewValue=<0.2 0 0 1>}
  n15 = FunctionInput() {bUsePreviewValueAsDefault=true, inputName=Max Pulse - Sphere Mult, inputType=FunctionInput_Scalar, sortPriority=60, previewValue=<1 0 0 1>}
  n16 = Multiply(
   A=Power(Base=OneMinus(None=@n10), Exp=LinearInterpolate(A=@n11, B=@n12, Alpha=@n13))
   B=LinearInterpolate(A=@n14, B=@n15, Alpha=NamedRerouteUsage())
  )
  FunctionOutput(None=Multiply(A=Saturate(None=@n16), B=NamedRerouteUsage())) {outputName="Sphere Mask (S)"}
  n17 = VectorParameter() {group=Emission Pulse TODM, parameterName=TOD - Color Morning, sortPriority=410, defaultValue=<rgba 1 1 1 1>}
  n18 = VectorParameter() {group=Emission Pulse TODM, parameterName=TOD - Color Day, sortPriority=420, defaultValue=<rgba 1 1 1 1>}
  n19 = VectorParameter() {group=Emission Pulse TODM, parameterName=TOD - Color Evening, sortPriority=430, defaultValue=<rgba 1 1 1 1>}
  n20 = VectorParameter() {group=Emission Pulse TODM, parameterName=TOD - Color Night, sortPriority=440, defaultValue=<rgba 1 1 1 1>}
  NamedRerouteDeclaration(None=MaterialFunctionCall(Color_Morning=@n17.RGB, Color_Day=@n18.RGB, Color_Evening=@n19.RGB, Color_Night=@n20.RGB) {materialFunction=&Color_Time_of_Day}.Result) {name="TOD Color (V3)", nodeColor=<rgba 0.5938 0 1 1>}
 ]}
