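# Character-style backpack material for the SwampRate back bling, adding a pulsing glowing core to skinned meshes.
# It wraps the shared CharacterParent 2023 PreFX and PostFX functions, overriding emissive between them through Set Material Attributes and named reroutes.
# MF_SwampRate_PulsingSphereMask builds a local-space sphere mask whose exponent and multiplier pulse, with a debug override.
# Glowing Core parameters select the emissive mask channel and control sphere placement, pulse and emission boost.
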Material /BRCosmetics/Accessories/FORT_Backpacks/Backpack_M_MED_SwampRate/Materials/M_Backpack_SwampRate
graph {
 properties {
  bAutomaticallySetUsageInEditor=false
  bUseMaterialAttributes=true
  bUsedWithMeshDeformer=true
  bUsedWithMorphTargets=true
  bUsedWithSkeletalMesh=true
  pixelDepthOffsetMode=PDOM_Legacy
  shadingModel=MSM_FromMaterialExpression
  usedShadingModels=Default Lit | Subsurface | Clear Coat | Subsurface Profile | Cloth
 }
 nodes [
  n1 = MaterialFunctionCall() {materialFunction=&MF_CharacterParent_2023_PreFX}
  MaterialFunctionCall(
   Material=SetMaterialAttributes(MaterialAttributes=NamedRerouteUsage(), Emissive Color=NamedRerouteUsage()) {attributeSetTypes [#1]}
   M=@n1.M
   Pre Skinned Local Position=@n1.M
   Distance Blend=@n1.M
   Fresnel=@n1.M
  ) {materialFunction=&MF_CharacterParent_2023_PostFX}
  n2 = ChannelMaskParameter(None=NamedRerouteUsage()) {group=Glowing Core, maskChannel=Green, parameterName=EmissiveCoreChannel, sortPriority=10, defaultValue=<rgba 0 1 0 0>}
  NamedRerouteDeclaration(None=@n2.R) {name="Emissive Core Mask (S)", nodeColor=<rgba 1 0 0.5391 1>}
  NamedRerouteDeclaration(None=@n1.Tech Art Mask) {name="TechArt Mask (V3)", nodeColor=<rgba 0.6609 0 1 1>}
  NamedRerouteDeclaration(None=GetMaterialAttributes(None=NamedRerouteDeclaration(None=@n1.Result) {name="Material In (MA)", nodeColor=<rgba 0 0.8594 1 1>}) {attributeGetTypes [#1]}.Emissive Color) {name="Emission In (V3)", nodeColor=<rgba 1 0.8203 0 1>}
  n3 = MaterialFunctionCall!(
   Sphere Center LS=NamedRerouteUsage()
   Sphere Radius=NamedRerouteUsage()
   Pulse Phase=NamedRerouteUsage()
   Min Pulse - Sphere Exp=NamedRerouteUsage()
   Max Pulse - Sphere Exp=NamedRerouteUsage()
   Min Pulse - Sphere Mult=NamedRerouteUsage()
   Max Pulse - Sphere Mult=NamedRerouteUsage()
   Pulse Override=NamedRerouteUsage()
  ) {materialFunction=/BRCosmetics/Accessories/FORT_Backpacks/Backpack_M_MED_SwampRate/Materials/MaterialFunctions/MF_SwampRate_PulsingSphereMask}
  n4 = Multiply(A=NamedRerouteUsage(), B=@n3."Sphere Mask (S)")
  n5 = NamedRerouteUsage()
  n6 = Multiply(A=@n5, B=NamedRerouteUsage())
  n7 = NamedRerouteUsage()
  NamedRerouteDeclaration(None=Multiply(A=Add(A=@n4), B=LinearInterpolate(A=@n5, B=@n6, Alpha=@n7))) {name="Emission Out (V3)", nodeColor=<rgba 0.2891 0 1 1>}
  NamedRerouteDeclaration(None=ScalarParameter() {defaultValue=0.15, group=Glowing Core, parameterName=Core - Emission Mult, sortPriority=20}) {name="Core - Emission Mult (S)", nodeColor=<rgba 0.7344 0 1 1>}
  n8 = VectorParameter() {
   group=Glowing Core
   parameterName="Sphere - Center (Local Space)"
   sortPriority=110
   channelNames {a=n/a, b=Center Z, g=Center Y, r=Center X}
   defaultValue=<rgba 0.0016 -13.43 -1.95 0>
  }
  NamedRerouteDeclaration(None=@n8.RGB) {name="Sphere Center (V3)", nodeColor=<rgba 0 1 0.4141 1>}
  NamedRerouteDeclaration(None=ScalarParameter() {defaultValue=15, group=Glowing Core, parameterName=Sphere - Radius, sortPriority=120}) {name="Sphere Radius (S)", nodeColor=<rgba 1 0.0938 0 1>}
  NamedRerouteDeclaration(None=ScalarParameter() {defaultValue=0.4, group=Glowing Core, parameterName=Pulse - Phase, sortPriority=220}) {name="Pulse - Phase (S)", nodeColor=<rgba 0 0.2266 1 1>}
  NamedRerouteDeclaration(None=ScalarParameter() {defaultValue=-1, group=Glowing Core, parameterName="Pulse - Override [Debug]", sortPriority=210}) {name="Pulse - Override (S)", nodeColor=<rgba 0.5469 1 0 1>}
  n9 = VectorParameter() {
   group=Glowing Core
   parameterName=Pulse - Sphere Controls
   sortPriority=230
   channelNames {a=Max Sphere Mult, b=Min Sphere Mult, g=Max Sphere Exp, r=Min Sphere Exp}
   defaultValue=<rgba 4 2.3 0.25 0.7>
  }
  NamedRerouteDeclaration(None=@n9.Min Sphere Exp) {name="Pulse - Min Sphere Exp (S)", nodeColor=<rgba 1 0 0.8672 1>}
  NamedRerouteDeclaration(None=@n9.Max Sphere Exp) {name="Pulse - Max Sphere Exp (S)", nodeColor=<rgba 1 0 0.8672 1>}
  NamedRerouteDeclaration(None=@n9.Min Sphere Mult) {name="Pulse - Min Sphere Mult (S)", nodeColor=<rgba 1 0 0.8672 1>}
  NamedRerouteDeclaration(None=@n9.Max Sphere Mult) {name="Pulse - Max Sphere Mult (S)", nodeColor=<rgba 1 0 0.8672 1>}
  NamedRerouteDeclaration(None=ScalarParameter() {defaultValue=40, group=Glowing Core, parameterName=Pulse - Max Additional Emission, sortPriority=240}) {name="Pulse - Max Additional Emission (S)", nodeColor=<rgba 0.7344 0 1 1>}
 ]
}
