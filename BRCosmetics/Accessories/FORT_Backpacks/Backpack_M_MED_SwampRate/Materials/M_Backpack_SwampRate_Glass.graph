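# Translucent glass material for the SwampRate backpack cosmetic, built between the shared 2023 character parent PreFX and PostFX functions.
# PreFX attributes blend toward a diffuse override color by that parameter's alpha, then feed the SwampRate glass function before PostFX.
# An emissive mask texture, remapped, multiplied and saturated, becomes the "Surface Etching Mask" named reroute.
# The glass function's Opacity Mask input also comes through a named reroute.
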Material /BRCosmetics/Accessories/FORT_Backpacks/Backpack_M_MED_SwampRate/Materials/M_Backpack_SwampRate_Glass
graph {
 properties {
  allowTranslucentCustomDepthWrites=true
  bAutomaticallySetUsageInEditor=false
  bScreenSpaceReflections=true
  bUseMaterialAttributes=true
  bUsedWithMeshDeformer=true
  bUsedWithMorphTargets=true
  bUsedWithSkeletalMesh=true
  blendMode=BLEND_Translucent
  shadingModel=MSM_FromMaterialExpression
  translucencyLightingMode=TLM_SurfacePerPixelLighting
  usedShadingModels=Default Lit | Subsurface | Clear Coat | Subsurface Profile | Cloth
 }
 nodes [
  n1 = VectorParameter() {
   group=Glass | 01_Surface
   parameterName=Surface - Diffuse Override
   sortPriority=60
   channelNames {a=Override Mix, b=Override B, g=Override G, r=Override R}
   defaultValue=<rgba 0.18 0.18 0.18 0>
  }
  n2 = MaterialFunctionCall() {materialFunction=&MF_CharacterParent_2023_PreFX}
  n3 = MaterialFunctionCall!(
   Material=BlendMaterialAttributes(
    A=@n2.Result
    B=SetMaterialAttributes!(MaterialAttributes=@n2.Result, Base Color=@n1.RGB) {attributeSetTypes [#1]}
    Alpha=Saturate(None=@n1.Override Mix)
   )
   Opacity Mask=NamedRerouteUsage()
  ) {materialFunction=/BRCosmetics/Accessories/FORT_Backpacks/Backpack_M_MED_SwampRate/Materials/MaterialFunctions/MF_SwampRate_Backpack_Glass}
  MaterialFunctionCall(Material=@n3."Result (MA)", M=@n2.M, Pre Skinned Local Position=@n2.M, Distance Blend=@n2.M, Fresnel=@n2.M) {materialFunction=&MF_CharacterParent_2023_PostFX}
  n4 = TextureSampleParameter2D() {
   group=06_Emissive
   parameterName=EmissiveMask
   samplerSource=SSM_Wrap_WorldGroupSettings
   samplerType=SAMPLERTYPE_Masks
   sortPriority=-100
   texture=&T_BlackMask
  }
  n5 = MaterialFunctionCall!(
   Input=@n4.R
   Input Low=ScalarParameter() {group=06_Emissive, parameterName=EmissiveMask OldMin, sliderMax=1, sortPriority=-90}
   Input High=ScalarParameter() {defaultValue=1, group=06_Emissive, parameterName=EmissiveMask OldMax, sliderMax=1, sortPriority=-80}
  ) {materialFunction=/Engine/Functions/Engine_MaterialFunctions03/Math/RemapValueRangeNormalized}
  n6 = ScalarParameter() {defaultValue=1, group=06_Emissive, parameterName=EmissiveMask Mult, sliderMax=1, sortPriority=-70}
  NamedRerouteDeclaration(None=Saturate(None=Multiply(A=@n5.Result, B=@n6))) {name="Surface Etching Mask (S)", nodeColor=<rgba 1 0 0.0703 1>}
 ]
}
