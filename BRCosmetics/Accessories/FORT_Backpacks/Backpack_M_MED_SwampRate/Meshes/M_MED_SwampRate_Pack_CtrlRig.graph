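# Control Rig for the SwampRate backpack cosmetic that drives the backpack_rotator bone under pack so it holds a world-relative tilt.
# Each evaluation converts an identity world transform into rig space and splits its rotation into ZYX Euler angles.
# It keeps X and Y, zeroes Z, and sets the bone's global rotation, countering component pitch and roll while leaving yaw attached to the component.
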ControlRigBlueprint /BRCosmetics/Accessories/FORT_Backpacks/Backpack_M_MED_SwampRate/Meshes/M_MED_SwampRate_Pack_CtrlRig
graphs [
 RigVMFunctionLibrary {links [], nodes []}
 RigVMModel {nodes [
   RigUnit_BeginExecution = &RigUnit_BeginExecution()
   From World = /Script/ControlRig.RigUnit_ToRigSpace_Transform(Value=(Rotation=(X=0,Y=0,Z=0,W=1),Translation=(X=0,Y=0,Z=0),Scale3D=(X=1,Y=1,Z=1))) {is=struct RigVMUnitNode, template="From World::Execute(in Value,out Global)"}
   RigVMFunction_MathQuaternionToEuler = &RigVMFunction_MathQuaternionToEuler(RotationOrder=ZYX, Value=@"From World".Global.Rotation)
   RigVMFunction_MathQuaternionFromEuler = &RigVMFunction_MathQuaternionFromEuler(Euler=(X=0,Y=0,Z=0), Euler.X=@RigVMFunction_MathQuaternionToEuler.Result.X, Euler.Y=@RigVMFunction_MathQuaternionToEuler.Result.Y)
   Set Transform = &RigUnit_SetRotation(
    Item=(Type=Bone,Name="backpack_rotator")
    Space=GlobalSpace
    Weight=1
    bInitial=False
    ExecutePin=@RigUnit_BeginExecution.ExecutePin
    Value=@RigVMFunction_MathQuaternionFromEuler.Result
   )
  ]}
]
hierarchy [
 root {loc=<0 0 0>, rot=<0 0 0 1>}
 pack {parent=root, loc=<0 -0.7316 0>, rot=<0 0 0 1>}
 backpack_rotator {parent=pack, loc=<0 -12.6688 0>, rot=<0 0 0.7071 -0.7071>}
]
