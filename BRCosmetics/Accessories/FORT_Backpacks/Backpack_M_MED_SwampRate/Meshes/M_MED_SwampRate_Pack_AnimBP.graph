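# Animation Blueprint for the M_MED_SwampRate backpack cosmetic, posing the pack skeleton.
# A Control Rig feeds a bone-space rotation replace on backpack_rotator, then three additive offsets on the pack bone with bool alphas.
# The event graph derives a continuous Ambient Rotation yaw from real time.
# On customization load it reads a custom part's gender and body type into flags, which each update combines into is M_Large for large males.
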AnimBlueprint /BRCosmetics/Accessories/FORT_Backpacks/Backpack_M_MED_SwampRate/Meshes/M_MED_SwampRate_Pack_AnimBP
target_skeleton=&/M_MED_SwampRate_Pack_Skeleton
graphs [
 AnimGraph {nodes [
   n1 = Control Rig(Source=Input Pose()) {controlRigAssetReference {blueprintRigClass=&/M_MED_SwampRate_Pack_CtrlRig.M_MED_SwampRate_Pack_CtrlRig_C}}
   n2 = "Transform (Modify) Bone - Bone: backpack_rotator"!(ComponentPose=Local To Component(LocalPose=@n1), Alpha=1, Rotation="0,0,0") -> (Pose) {
    rotationMode=BMM_Replace
    rotationSpace=BCS_BoneSpace
    translationSpace=BCS_WorldSpace
    alphaBoolBlend {blendOption=Linear}
    boneToModify {boneName=backpack_rotator}
   }
   n3 = "Transform (Modify) Bone - Bone: pack"!(ComponentPose=@n2, bAlphaBoolEnabled:bool, Translation="0,-1.75,0") -> (Pose) {
    alphaInputType=Bool
    translationMode=BMM_Additive
    translationSpace=BCS_BoneSpace
    alphaBoolBlend {blendOption=Linear}
    boneToModify {boneName=pack}
    translation {y=-1.75}
   }
   n4 = "Transform (Modify) Bone - Bone: pack"(ComponentPose=@n3, Translation="0,-0.5,0") {
    alphaInputType=Bool
    translationMode=BMM_Additive
    translationSpace=BCS_BoneSpace
    alphaBoolBlend {blendOption=Linear}
    boneToModify {boneName=pack}
    translation {y=-0.5}
   }
   n5 = "Transform (Modify) Bone - Bone: pack"(ComponentPose=@n4, Translation="0,-2,0") {
    alphaInputType=Bool
    translationMode=BMM_Additive
    translationSpace=BCS_BoneSpace
    alphaBoolBlend {blendOption=Linear}
    boneToModify {boneName=pack}
    translation {y=-2}
   }
   Output Pose(Result=Component To Local(ComponentPose=@n5)) {layerGroup=None}
  ]}
 EventGraph {nodes [
   Event BlueprintUpdateAnimation()
   n1 = float / float(A=GetRealTimeSeconds(), B=-500) {is=title K2Node_CallFunction}
   > Set Ambient Rotation(Ambient Rotation:Rotator=MakeRotator(Yaw=float * float(A=@n1, B=10000))) -> (Output_Get:Rotator)
   > Set is M_Large(is M_Large:bool=AND Boolean(A=Get isMale() -> (isMale:bool), B=Get isLarge() -> (isLarge:bool))) -> (Output_Get:bool)
   n2 = Event OnCharacterCustomizationLoaded()
   n3 = Get Array Item with EFort Custom Part Type(Array:Array of Custom Character Part Object References=@n2.CustomParts) -> (Output:Custom Character Part Object Reference)
   n4 = Get GenderPermitted!(self:Custom Character Part Object Reference=@n3) -> (GenderPermitted:EFortCustomGender Enum=Invalid)
   > Set isMale(isMale:bool="Equal (Enum)"(A:EFortCustomGender Enum=@n4, B:EFortCustomGender Enum=Male)) -> (Output_Get:bool)
   > Set isFemale(isFemale:bool="Equal (Enum)"(A=@n4, B=Female)) -> (Output_Get:bool)
   > Set isLarge(isLarge:bool="Equal (Enum)"(
     A:EFortCustomBodyType Enum=Get BodyTypesPermitted!(self:Custom Character Part Object Reference=@n3) -> (BodyTypesPermitted:EFortCustomBodyType Enum=NONE)
     B:EFortCustomBodyType Enum=Large
    )) -> (Output_Get:bool)
   Event BlueprintInitializeAnimation()
   > Set Skel_Mesh(Skel_Mesh=GetOwnerSkeletalMeshComponent())
   TryGetPawnOwner()
  ]}
]
variables [
 Ambient Rotation {type=Rotator}
 is M_Large {type=boolean, default=true}
 isFemale {type=boolean}
 isLarge {type=boolean}
 isMale {type=boolean}
 Skel_Mesh {type=SkeletalMeshComponent}
]
